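; Held weapon Blueprint for the Zero Point sprite, handling how the sprite creature appears while equipped.
; On equip and attach it casts the weapon mesh's anim instance to the Zero Point creature anim Blueprint and flags the equipped idle state.
; After a short delay it shows the mesh and activates the idle super-sprite Niagara effect.
; Local unequip hides the mesh and deactivates that effect.
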(event FortWeapon|EventOnInstigatorSet
  (|CallParentFunction))

(event FortWeapon|EventOnLocalUnEquip
  (Rendering|SetVisibility (Variables|Visual|GetWeaponMesh) false false)
  (Components|Activation|Deactivate (Variables|Default|GetNS_SuperSprite_Idle)))

(event FortWeapon|EventOnEquippedAndAttached
  (|CallParentFunction)
  (Utilities|FlowControl|Sequence
    (:then_0
      (bind _assprite_zero_point_creature_skeleton_anim_blueprint (Utilities|Casting|CastToSprite_ZeroPoint_Creature_Skeleton_AnimBlueprint (Components|SkeletalMesh|GetAnimInstance (Variables|Visual|GetWeaponMesh))))
      (Variables|/Script/SpritesRuntime.SpriteCreatureAnimInstance|SetbIsEquipped_Idle true _assprite_zero_point_creature_skeleton_anim_blueprint))
    (:then_1
      (Utilities|FlowControl|Delay 0.1)
      (Rendering|SetVisibility (Variables|Visual|GetWeaponMesh) true false)
      (Components|Activation|Activate (Variables|Default|GetNS_SuperSprite_Idle) false))))

(fn ConstructionScript)
